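# Sound cue for a one-shot evening meadowlark call in the grassland ambience of the Chrome Coney Crossroads POI.
# The top two quality levels play a varied random call alone or, less often, layered with a delayed, quieter, lower-pitched second call.
# The other two levels play a single modulated call from a smaller sample pool.
# Bird one-shot attenuation and a chrome animal effect chain borrowed from HerraldsFortress shape the output.
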SoundCue /Game/Sounds/Artemis/POI/Chrome_ConeyCrossroads/Spots/Spot_Chrome_Grassland_Evening_Meadowlark_Call_Cue
root=QualityLevel_0
nodes [
 {id=Delay_0, children [Modulator_0], properties {delayMax=2.5, delayMin=2}}
 {id=Mixer_0, children [Modulator_1,Delay_0], properties {inputVolume [1,0.25]}}
 {id=Modulator_0, children [Random_4], properties {pitchMax=0.9, pitchMin=0.85, volumeMax=0.6, volumeMin=0.5}}
 {id=Modulator_1, children [Random_2], properties {pitchMax=1.1, volumeMax=1.1, volumeMin=0.9}}
 {id=Modulator_2, children [Random_3], properties {pitchMax=1.1, volumeMax=1.1, volumeMin=0.9}}
 {id=QualityLevel_0, children [Random_5,Random_5,Modulator_2,Modulator_2]}
 {id=Random_2, children [WavePlayer_8, WavePlayer_9, WavePlayer_10, WavePlayer_11], properties {weights [1,1,1,1]}}
 {id=Random_3, children [WavePlayer_12, WavePlayer_13, WavePlayer_14], properties {weights [1,1,1]}}
 {id=Random_4, children [WavePlayer_15, WavePlayer_16, WavePlayer_17, WavePlayer_18], properties {weights [1,1,1,1]}}
 {id=Random_5, children [Modulator_1,Mixer_0], properties {weights [1,0.33]}}
 {
  id=WavePlayer_10
  properties {soundWaveAssetPtr=/Game/Sounds/Artemis/Biomes/Grasslands/Grassland_Spots/Spot_Grassland_Evening_Meadowlark_Call_03}
 }
 {
  id=WavePlayer_11
  properties {soundWaveAssetPtr=/Game/Sounds/Artemis/Biomes/Grasslands/Grassland_Spots/Spot_Grassland_Evening_Meadowlark_Call_04}
 }
 {
  id=WavePlayer_12
  properties {soundWaveAssetPtr=&Spot_Grassland_Evening_Meadowlark_Call_04}
 }
 {
  id=WavePlayer_13
  properties {soundWaveAssetPtr=&Spot_Grassland_Evening_Meadowlark_Call_03}
 }
 {
  id=WavePlayer_14
  properties {soundWaveAssetPtr=/Game/Sounds/Artemis/Biomes/Grasslands/Grassland_Spots/Spot_Grassland_Evening_Meadowlark_Call_02}
 }
 {
  id=WavePlayer_15
  properties {soundWaveAssetPtr=/Game/Sounds/Artemis/Biomes/Grasslands/Grassland_Spots/Spot_Grassland_Evening_Meadowlark_Call_01}
 }
 {
  id=WavePlayer_16
  properties {soundWaveAssetPtr=&Spot_Grassland_Evening_Meadowlark_Call_02}
 }
 {
  id=WavePlayer_17
  properties {soundWaveAssetPtr=&Spot_Grassland_Evening_Meadowlark_Call_03}
 }
 {
  id=WavePlayer_18
  properties {soundWaveAssetPtr=&Spot_Grassland_Evening_Meadowlark_Call_04}
 }
 {
  id=WavePlayer_8
  properties {soundWaveAssetPtr=&Spot_Grassland_Evening_Meadowlark_Call_01}
 }
 {
  id=WavePlayer_9
  properties {soundWaveAssetPtr=&Spot_Grassland_Evening_Meadowlark_Call_02}
 }
]
properties {
 attenuationSettings=&Fort_TileAmbient_OneShot_Bird_Att
 bEnableSubmixSends=false
 maxDistance=8005
 soundClassObject=&Fort_Ambient_OneShots
 soundSubmixObject=&Fortnite_Reverb_Distant
 sourceEffectChain=/Game/Sounds/Artemis/POI/HerraldsFortress/Spots/Spot_Chrome_Animal_FXChain
 volumeMultiplier=0.301
 preEffectBusSends [{minSendLevel=0.2}]
}
