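# Sound cue for an evening raven call ambient one-shot at the Chrome Coney Crossroads POI, using grassland recordings through a Chrome animal effect chain.
# The first two quality slots usually play one of four varied calls, occasionally mixing in a delayed, quieter, lower-pitched second call.
# The other two slots play only a single modulated call from three recordings.
# Bird one-shot attenuation and a distant reverb submix are assigned, with submix sends disabled.
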SoundCue /Game/Sounds/Artemis/POI/Chrome_ConeyCrossroads/Spots/Spot_Chrome_Grassland_Evening_Raven_Call_Cue
root=QualityLevel_0
nodes [
 {id=Delay_0, children [Modulator_0], properties {delayMax=2.5, delayMin=2}}
 {id=Mixer_0, children [Modulator_1,Delay_0], properties {inputVolume [1,0.5]}}
 {id=Modulator_0, children [Random_1], properties {pitchMax=0.9, pitchMin=0.8, volumeMax=0.5, volumeMin=0.4}}
 {id=Modulator_1, children [Random_0], properties {pitchMax=1.1, volumeMax=1, volumeMin=0.8}}
 {id=Modulator_2, children [Random_3], properties {pitchMax=1.1, volumeMax=1, volumeMin=0.8}}
 {id=QualityLevel_0, children [Random_5,Random_5,Modulator_2,Modulator_2]}
 {id=Random_0, children [WavePlayer_0,WavePlayer_1,WavePlayer_2,WavePlayer_3], properties {weights [1,1,1,1]}}
 {id=Random_1, children [WavePlayer_4,WavePlayer_5,WavePlayer_6,WavePlayer_7], properties {weights [1,1,1,1]}}
 {id=Random_3, children [WavePlayer_19, WavePlayer_20, WavePlayer_21], properties {weights [1,1,1]}}
 {id=Random_5, children [Modulator_1,Mixer_0], properties {weights [1,0.33]}}
 {id=WavePlayer_0, properties {soundWaveAssetPtr=/Game/Sounds/Artemis/Biomes/Grasslands/Grassland_Spots/Spot_Grassland_Evening_Raven_Call_01}}
 {id=WavePlayer_1, properties {soundWaveAssetPtr=/Game/Sounds/Artemis/Biomes/Grasslands/Grassland_Spots/Spot_Grassland_Evening_Raven_Call_02}}
 {id=WavePlayer_19, properties {soundWaveAssetPtr=&Spot_Grassland_Evening_Raven_Call_01}}
 {id=WavePlayer_2, properties {soundWaveAssetPtr=/Game/Sounds/Artemis/Biomes/Grasslands/Grassland_Spots/Spot_Grassland_Evening_Raven_Call_03}}
 {id=WavePlayer_20, properties {soundWaveAssetPtr=&Spot_Grassland_Evening_Raven_Call_02}}
 {id=WavePlayer_21, properties {soundWaveAssetPtr=&Spot_Grassland_Evening_Raven_Call_03}}
 {id=WavePlayer_3, properties {soundWaveAssetPtr=/Game/Sounds/Artemis/Biomes/Grasslands/Grassland_Spots/Spot_Grassland_Evening_Raven_Call_04}}
 {id=WavePlayer_4, properties {soundWaveAssetPtr=&Spot_Grassland_Evening_Raven_Call_01}}
 {id=WavePlayer_5, properties {soundWaveAssetPtr=&Spot_Grassland_Evening_Raven_Call_03}}
 {id=WavePlayer_6, properties {soundWaveAssetPtr=&Spot_Grassland_Evening_Raven_Call_02}}
 {id=WavePlayer_7, properties {soundWaveAssetPtr=&Spot_Grassland_Evening_Raven_Call_04}}
]
properties {
 attenuationSettings=&Fort_TileAmbient_OneShot_Bird_Att
 bEnableSubmixSends=false
 maxDistance=8005
 soundClassObject=&Fort_Ambient_OneShots
 soundSubmixObject=&Fortnite_Reverb_Distant
 sourceEffectChain=/Game/Sounds/Artemis/POI/HerraldsFortress/Spots/Spot_Chrome_Animal_FXChain
 volumeMultiplier=0.401
 preEffectBusSends [{minSendLevel=0.2}]
}
