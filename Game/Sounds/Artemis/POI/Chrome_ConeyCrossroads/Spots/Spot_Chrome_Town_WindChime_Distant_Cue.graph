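# Ambient one-shot cue for distant town wind chimes at the Chrome Coney Crossroads POI, using tile-ambient attenuation and the distant reverb submix.
# Each mix layers random close and distant chime waves, with pitch/volume modulation and envelopes fading out by three seconds.
# A quality switch gives the first two levels larger wave pools and the last two a reduced set.
# Its source effect chain is borrowed from a HerraldsFortress animal spot.
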SoundCue /Game/Sounds/Artemis/POI/Chrome_ConeyCrossroads/Spots/Spot_Chrome_Town_WindChime_Distant_Cue
root=QualityLevel_0
nodes [
 {
  id=Enveloper_0
  children [Modulator_0]
  properties {
   pitchCurve {editorCurveData {keys [{time=0}]}}
   volumeCurve {editorCurveData {keys [<at 0 1 RCIM_Cubic>, {arriveTangent=0.0054, interpMode=RCIM_Cubic, leaveTangent=0.0054, tangentMode=RCTM_User, time=1.5, value=1}, <at 3 0>]}}
  }
 }
 {
  id=Enveloper_1
  children [Modulator_1]
  properties {
   pitchCurve {editorCurveData {keys [{time=0}]}}
   volumeCurve {editorCurveData {keys [
      {arriveTangent=0.0098, interpMode=RCIM_Cubic, leaveTangent=0.0098, tangentMode=RCTM_User, time=0, value=1}
      {arriveTangent=0.0054, interpMode=RCIM_Cubic, leaveTangent=0.0054, tangentMode=RCTM_User, time=1, value=1}
      <at 3 0>
     ]}}
  }
 }
 {
  id=Enveloper_4
  children [Modulator_5]
  properties {
   pitchCurve {editorCurveData {keys [{time=0}]}}
   volumeCurve {editorCurveData {keys [<at 0 1 RCIM_Cubic>, {arriveTangent=0.0054, interpMode=RCIM_Cubic, leaveTangent=0.0054, tangentMode=RCTM_User, time=1.5, value=1}, <at 3 0>]}}
  }
 }
 {
  id=Enveloper_5
  children [Modulator_4]
  properties {
   pitchCurve {editorCurveData {keys [{time=0}]}}
   volumeCurve {editorCurveData {keys [
      {arriveTangent=0.0098, interpMode=RCIM_Cubic, leaveTangent=0.0098, tangentMode=RCTM_User, time=0, value=1}
      {arriveTangent=0.0054, interpMode=RCIM_Cubic, leaveTangent=0.0054, tangentMode=RCTM_User, time=1, value=1}
      <at 3 0>
     ]}}
  }
 }
 {id=Mixer_0, children [Enveloper_0,Enveloper_1], properties {inputVolume [0.5,1]}}
 {id=Mixer_1, children [Enveloper_4,Enveloper_5], properties {inputVolume [1,1]}}
 {id=Modulator_0, children [Random_4], properties {pitchMax=1.1, pitchMin=1, volumeMax=0.6, volumeMin=0.5}}
 {id=Modulator_1, children [Random_0], properties {pitchMax=1.1, volumeMax=1.1, volumeMin=0.9}}
 {id=Modulator_4, children [Random_8], properties {pitchMax=1.1, volumeMax=1.1, volumeMin=0.9}}
 {id=Modulator_5, children [Random_3], properties {pitchMax=1.1, pitchMin=1, volumeMax=0.6, volumeMin=0.5}}
 {id=QualityLevel_0, children [Mixer_0,Mixer_0,Mixer_1,Mixer_1]}
 {
  id=Random_0
  children [WavePlayer_0, WavePlayer_1, WavePlayer_2, WavePlayer_3, WavePlayer_4, WavePlayer_13]
  properties {weights [1,1,1,1,1,1]}
 }
 {id=Random_3, children [WavePlayer_14, WavePlayer_15, WavePlayer_16], properties {weights [1,1,1]}}
 {
  id=Random_4
  children [
   WavePlayer_17
   WavePlayer_18
   WavePlayer_19
   WavePlayer_20
   WavePlayer_21
   WavePlayer_22
   WavePlayer_23
   WavePlayer_24
   WavePlayer_25
   WavePlayer_26
  ]
  properties {weights [1,1,1,1,1,1,1,1,1,1]}
 }
 {id=Random_8, children [WavePlayer_28, WavePlayer_29, WavePlayer_30, WavePlayer_31], properties {weights [1,1,1,1]}}
 {id=WavePlayer_0, properties {soundWaveAssetPtr=/Game/Sounds/Artemis/Biomes/Towns/Town_Spots/Spot_Town_WindChime_Distant_01}}
 {id=WavePlayer_1, properties {soundWaveAssetPtr=/Game/Sounds/Artemis/Biomes/Towns/Town_Spots/Spot_Town_WindChime_Distant_02}}
 {id=WavePlayer_13, properties {soundWaveAssetPtr=/Game/Sounds/Artemis/Biomes/Towns/Town_Spots/Spot_Town_WindChime_Distant_06}}
 {id=WavePlayer_14, properties {soundWaveAssetPtr=&Spot_Town_WindChime_Distant_01}}
 {id=WavePlayer_15, properties {soundWaveAssetPtr=/Game/Sounds/Artemis/Biomes/Towns/Town_Spots/Spot_Town_WindChime_Distant_03}}
 {id=WavePlayer_16, properties {soundWaveAssetPtr=&Spot_Town_WindChime_Distant_06}}
 {id=WavePlayer_17, properties {soundWaveAssetPtr=/Game/Sounds/Artemis/Biomes/Towns/Town_Spots/Spot_Town_Wind_Chimes_10}}
 {id=WavePlayer_18, properties {soundWaveAssetPtr=/Game/Sounds/Artemis/Biomes/Towns/Town_Spots/Spot_Town_Wind_Chimes_01}}
 {id=WavePlayer_19, properties {soundWaveAssetPtr=/Game/Sounds/Artemis/Biomes/Towns/Town_Spots/Spot_Town_Wind_Chimes_02}}
 {id=WavePlayer_2, properties {soundWaveAssetPtr=&Spot_Town_WindChime_Distant_03}}
 {id=WavePlayer_20, properties {soundWaveAssetPtr=/Game/Sounds/Artemis/Biomes/Towns/Town_Spots/Spot_Town_Wind_Chimes_03}}
 {id=WavePlayer_21, properties {soundWaveAssetPtr=/Game/Sounds/Artemis/Biomes/Towns/Town_Spots/Spot_Town_Wind_Chimes_04}}
 {id=WavePlayer_22, properties {soundWaveAssetPtr=/Game/Sounds/Artemis/Biomes/Towns/Town_Spots/Spot_Town_Wind_Chimes_05}}
 {id=WavePlayer_23, properties {soundWaveAssetPtr=/Game/Sounds/Artemis/Biomes/Towns/Town_Spots/Spot_Town_Wind_Chimes_06}}
 {id=WavePlayer_24, properties {soundWaveAssetPtr=/Game/Sounds/Artemis/Biomes/Towns/Town_Spots/Spot_Town_Wind_Chimes_07}}
 {id=WavePlayer_25, properties {soundWaveAssetPtr=/Game/Sounds/Artemis/Biomes/Towns/Town_Spots/Spot_Town_Wind_Chimes_08}}
 {id=WavePlayer_26, properties {soundWaveAssetPtr=/Game/Sounds/Artemis/Biomes/Towns/Town_Spots/Spot_Town_Wind_Chimes_09}}
 {id=WavePlayer_28, properties {soundWaveAssetPtr=&Spot_Town_Wind_Chimes_10}}
 {id=WavePlayer_29, properties {soundWaveAssetPtr=&Spot_Town_Wind_Chimes_01}}
 {id=WavePlayer_3, properties {soundWaveAssetPtr=/Game/Sounds/Artemis/Biomes/Towns/Town_Spots/Spot_Town_WindChime_Distant_04}}
 {id=WavePlayer_30, properties {soundWaveAssetPtr=&Spot_Town_Wind_Chimes_02}}
 {id=WavePlayer_31, properties {soundWaveAssetPtr=&Spot_Town_Wind_Chimes_03}}
 {id=WavePlayer_4, properties {soundWaveAssetPtr=/Game/Sounds/Artemis/Biomes/Towns/Town_Spots/Spot_Town_WindChime_Distant_05}}
]
properties {
 attenuationSettings=/Game/Sounds/Attenuation/Fort_TileAmbient_OneShot_Att
 bEnableSubmixSends=false
 maxDistance=8000
 soundClassObject=&Fort_Ambient_Athena
 soundSubmixObject=&Fortnite_Reverb_Distant
 sourceEffectChain=/Game/Sounds/Artemis/POI/HerraldsFortress/Spots/Spot_Chrome_Animal_FXChain
 volumeMultiplier=0.802
 preEffectBusSends [{minSendLevel=0.2}]
}
