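# Looping ambient spot SoundCue for Coney Crossroads that plays intermittent light metal chain movement sounds.
# A loop repeats a randomized delay, then picks one of six equally weighted Tilted Towers chain waves through a pitch-varying modulator.
# A quality-level node feeds that branch to its first two slots and leaves the other two empty.
# It uses the ambient sound class and references an AirborneCity linear attenuation asset.
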SoundCue /Game/Sounds/Artemis/POI/Chrome_ConeyCrossroads/Spots/Spot_ConeyCrossRoads_Metal_Chain_Movement_Light_Looping_LinearIntExt_Cue
root=Looping_0
nodes [
 {id=Delay_0, children [QualityLevel_0], properties {delayMax=12, delayMin=3}}
 {id=Looping_0, children [Delay_0]}
 {id=Modulator_0, children [Random_2], properties {pitchMin=0.9}}
 {id=QualityLevel_0, children [Modulator_0,Modulator_0,null,null]}
 {id=Random_2, children [WavePlayer_3,WavePlayer_4,WavePlayer_5,WavePlayer_6,WavePlayer_7,WavePlayer_8], properties {weights [1,1,1,1,1,1]}}
 {id=WavePlayer_3, properties {soundWaveAssetPtr=/Game/Sounds/Artemis/POI/Tilted_Towers/Spots/Spot_Tilted_Towers_Chain_Movement_Light_01}}
 {id=WavePlayer_4, properties {soundWaveAssetPtr=/Game/Sounds/Artemis/POI/Tilted_Towers/Spots/Spot_Tilted_Towers_Chain_Movement_Light_02}}
 {id=WavePlayer_5, properties {soundWaveAssetPtr=/Game/Sounds/Artemis/POI/Tilted_Towers/Spots/Spot_Tilted_Towers_Chain_Movement_Light_03}}
 {id=WavePlayer_6, properties {soundWaveAssetPtr=/Game/Sounds/Artemis/POI/Tilted_Towers/Spots/Spot_Tilted_Towers_Chain_Movement_Light_04}}
 {id=WavePlayer_7, properties {soundWaveAssetPtr=/Game/Sounds/Artemis/POI/Tilted_Towers/Spots/Spot_Tilted_Towers_Chain_Movement_Light_05}}
 {id=WavePlayer_8, properties {soundWaveAssetPtr=/Game/Sounds/Artemis/POI/Tilted_Towers/Spots/Spot_Tilted_Towers_Chain_Movement_Light_06}}
]
properties {
 attenuationSettings=/Game/Sounds/Artemis/POI/AirborneCity/Spots/Spot_Wind_Vibrating_Metal_Structure_Small_Linear_Att
 maxDistance=1850
 soundClassObject=&Fort_Ambient_Athena
 volumeMultiplier=1.253
 attenuationOverrides {falloffDistance=6000, stereoSpread=10000, attenuationShapeExtents {x=2000}}
}
