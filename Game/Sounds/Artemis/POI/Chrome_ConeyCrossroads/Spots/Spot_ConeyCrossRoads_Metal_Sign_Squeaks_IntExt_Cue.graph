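# Sound cue for spot ambient metal sign squeaks at the Coney Crossroads POI, played as a positional one-shot in the ambient sound class.
# A random node picks one of six equally weighted metal sign movement waves, all reused from the Tilted Towers spot sounds.
# A modulator varies pitch on each play, and a shared tile-ambient attenuation asset is paired with reduced cue volume.
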SoundCue /Game/Sounds/Artemis/POI/Chrome_ConeyCrossroads/Spots/Spot_ConeyCrossRoads_Metal_Sign_Squeaks_IntExt_Cue
root=Modulator_0
nodes [
 {id=Modulator_0, children [Random_0], properties {pitchMin=0.9}}
 {
  id=Random_0
  children [WavePlayer_3, WavePlayer_4, WavePlayer_5, WavePlayer_6, WavePlayer_7, WavePlayer_11]
  properties {weights [1,1,1,1,1,1]}
 }
 {id=WavePlayer_11, properties {soundWaveAssetPtr=/Game/Sounds/Artemis/POI/Tilted_Towers/Spots/Spot_Tilted_Towers_Sign_Metal_Movement_04}}
 {id=WavePlayer_3, properties {soundWaveAssetPtr=/Game/Sounds/Artemis/POI/Tilted_Towers/Spots/Spot_Tilted_Towers_Sign_Metal_Movement_06}}
 {id=WavePlayer_4, properties {soundWaveAssetPtr=/Game/Sounds/Artemis/POI/Tilted_Towers/Spots/Spot_Tilted_Towers_Sign_Metal_Movement_03}}
 {id=WavePlayer_5, properties {soundWaveAssetPtr=/Game/Sounds/Artemis/POI/Tilted_Towers/Spots/Spot_Tilted_Towers_Sign_Metal_Movement_05}}
 {id=WavePlayer_6, properties {soundWaveAssetPtr=/Game/Sounds/Artemis/POI/Tilted_Towers/Spots/Spot_Tilted_Towers_Sign_Metal_Movement_01}}
 {id=WavePlayer_7, properties {soundWaveAssetPtr=/Game/Sounds/Artemis/POI/Tilted_Towers/Spots/Spot_Tilted_Towers_Sign_Metal_Movement_02}}
]
properties {
 attenuationSettings=/Game/Sounds/Attenuation/Fort_TileAmbient_OneShot_LessPositional_Log_Att
 maxDistance=6200
 soundClassObject=&Fort_Ambient_Athena
 volumeMultiplier=0.601
 attenuationOverrides {falloffDistance=6000, stereoSpread=10000, attenuationShapeExtents {x=2000}}
}
